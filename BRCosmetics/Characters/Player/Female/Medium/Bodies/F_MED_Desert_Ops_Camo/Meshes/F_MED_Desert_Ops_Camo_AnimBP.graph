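# Post-processing AnimBlueprint for the female medium Desert Ops Camo body cosmetic, copying the attached parent mesh's pose and layering bone corrections on it.
# Thigh-rotation alphas drive additive corrections on both dyn_glute bones.
# Weighted constraints tie attach_backpack and attach_cape to four spine bones, favoring spine_05.
# When customization loads, vest meta tags set VestAlpha, which blends in offsets on both attach bones to accommodate the vest.
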AnimBlueprint /BRCosmetics/Characters/Player/Female/Medium/Bodies/F_MED_Desert_Ops_Camo/Meshes/F_MED_Desert_Ops_Camo_AnimBP
target_skeleton=/Game/Characters/Player/Female/Medium/Bodies/F_MED_Desert_Ops_Camo/Meshes/F_MED_Desert_Ops_Camo_Skeleton
graphs [
 AnimGraph {nodes [
   n1 = "Transform (Modify) Bone - Bone: dyn_glute_l"(
    ComponentPose=Local To Component(LocalPose=Copy Pose From Mesh() {bUseAttachedParent=true})
    Alpha=Get ThighRotLeftAlpha() -> (ThighRotLeftAlpha:float)
    Translation="-3.750,0.875,-0.9"
    Rotation="0.000,0.000,-55"
    Scale="1.25,1.25,1.25"
   ) {
    rotationMode=BMM_Additive
    rotationSpace=BCS_BoneSpace
    scaleMode=BMM_Additive
    scaleSpace=BCS_BoneSpace
    translationMode=BMM_Additive
    translationSpace=BCS_BoneSpace
    alphaBoolBlend {blendOption=Linear}
    boneToModify {boneName=dyn_glute_l}
    rotation {roll=-55}
    scale=<1.25 1.25 1.25>
    translation=<-3.75 0.875 -0.9>
   }
   n2 = "Transform (Modify) Bone - Bone: dyn_glute_r"(
    ComponentPose=@n1
    Alpha=Get ThighRotRightAlpha() -> (ThighRotRightAlpha:float)
    Translation="-3.75,0,0"
    Rotation="0.000,0.000,-55"
    Scale="1.3,1.3,1.3"
   ) {
    rotationMode=BMM_Additive
    rotationSpace=BCS_BoneSpace
    scaleMode=BMM_Additive
    scaleSpace=BCS_BoneSpace
    translationMode=BMM_Additive
    translationSpace=BCS_BoneSpace
    alphaBoolBlend {blendOption=Linear}
    boneToModify {boneName=dyn_glute_r}
    rotation {roll=-55}
    scale=<1.3 1.3 1.3>
    translation {x=-3.75}
   }
   n3 = "Constraint - attach_backpack (PPPP)"!(
    ComponentPose=@n2
    Alpha=1
    ConstraintWeights_0=0.0173
    ConstraintWeights_1:float=0.0303
    ConstraintWeights_2:float=0.2149
    ConstraintWeights_3:float=0.7375
   ) -> (Pose) {
    alphaBoolBlend {blendOption=Linear}
    boneToModify {boneName=attach_backpack}
    constraintSetup [{targetBone {boneName=spine_02}}, {targetBone {boneName=spine_03}}, {targetBone {boneName=spine_04}}, {targetBone {boneName=spine_05}}]
    constraintWeights [1,1,1,1]
   }
   n4 = "Constraint - attach_cape (PPPP)"!(
    ComponentPose=@n3
    Alpha=1
    ConstraintWeights_0=0.0173
    ConstraintWeights_1:float=0.0303
    ConstraintWeights_2:float=0.2149
    ConstraintWeights_3:float=0.7375
   ) -> (Pose) {
    alphaBoolBlend {blendOption=Linear}
    boneToModify {boneName=attach_cape}
    constraintSetup [{targetBone {boneName=spine_02}}, {targetBone {boneName=spine_03}}, {targetBone {boneName=spine_04}}, {targetBone {boneName=spine_05}}]
    constraintWeights [1,1,1,1]
   }
   n5 = Get VestAlpha() -> (VestAlpha:double)
   n6 = "Transform (Modify) Bone - Bone: attach_backpack"(ComponentPose=@n4, Alpha=@n5, Translation="0.000,-3,0.000", Rotation="0,0,0", Scale="1,1,1") {
    rotationMode=BMM_Additive
    rotationSpace=BCS_BoneSpace
    translationMode=BMM_Additive
    translationSpace=BCS_BoneSpace
    alphaBoolBlend {blendOption=Linear}
    boneToModify {boneName=attach_backpack}
    translation {y=-3}
   }
   n7 = "Transform (Modify) Bone - Bone: attach_cape"(ComponentPose=@n6, Alpha=@n5, Translation="0.000,-3,-1.85", Rotation="0,0,0", Scale="1,1,1") {
    rotationMode=BMM_Additive
    rotationSpace=BCS_BoneSpace
    translationMode=BMM_Additive
    translationSpace=BCS_BoneSpace
    alphaBoolBlend {blendOption=Linear}
    boneToModify {boneName=attach_cape}
    translation {y=-3, z=-1.85}
   }
   Output Pose(Result=Component To Local(ComponentPose=@n7)) {layerGroup=None}
  ]}
 EventGraph {nodes [
   Event OnCharacterCustomizationLoaded()
   > n1 = Branch(Condition=CharacterItemDefinitionHasAnyMetaTag(Tags=Make Array!(
      "[0]"=(TagName="Cosmetics.MetaTags.Vest.01")
      "[1]":Gameplay Tag Structure=(TagName="Cosmetics.MetaTags.Vest.02")
      "[2]":Gameplay Tag Structure=(TagName="Cosmetics.MetaTags.Vest.03")
      "[3]":Gameplay Tag Structure=(TagName="Cosmetics.MetaTags.Vest.04")
     ) -> (Array)))
   > Set VestAlpha(VestAlpha:double=1) -> (Output_Get:double)
   Set VestAlpha(execute=@n1.else, VestAlpha=0)
  ]}
]
variables [VestAlpha {type=number}]
